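; Behavior tree decorator that passes when the AI's controlled pawn has an active ability matching any of a configured set of gameplay tags.
; The AI condition check gets the pawn's Fort ability system component and queries it with the ActiveAbilityTagsToCheck container.
; The query runs only if the pawn and component are valid and the tag container is not empty; otherwise the condition does not pass.
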(fn PerformConditionCheckAI
  (params (OwnerController :type AIController%20Object%20Reference) (ControlledPawn :type Pawn%20Object%20Reference))
  (bind _returnvalue (Ability|GetFortAbilitySystemComponent ControlledPawn))
  (bind _activeabilitytagstocheck (Variables|Default|GetActiveAbilityTagsToCheck))
  (Utilities|IsValid ControlledPawn
    (:Is%20Valid
      (Utilities|IsValid _returnvalue
        (:Is%20Valid
          (if (GameplayTags|NotEqual(GameplayTagContainer) _activeabilitytagstocheck (GameplayTags|MakeLiteralGameplayTagContainer))
            (return (Abilities|HasActiveAbilitywithAnyMatchingTag _returnvalue _activeabilitytagstocheck))
            (else
              _)))
        (:Is%20Not%20Valid
          _))
      (return false))
    (:Is%20Not%20Valid
      _)))
