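; Behavior tree decorator for Athena AI that tests whether the actor in a configurable blackboard key is in water.
; It first requires a valid controlled pawn, ability system component and target actor, then branches by cast on the target's type.
; Player pawns pass when surface swimming or in a vehicle that is in water, and other Fort pawns only when surface swimming.
; Pickups and projectiles pass when overlapping a water body actor, while other types fail.
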(fn PerformConditionCheckAI
  (params (OwnerController :type AIController%20Object%20Reference) (ControlledPawn :type Pawn%20Object%20Reference))
  (bind _blackboardkeytotestisinwater (Variables|IsInWaterSetup|GetBlackboardKeyToTestIsInWater))
  (bind _returnvalue (AI|BehaviorTree|GetBlackboardValueasActor _blackboardkeytotestisinwater))
  (Utilities|IsValid ControlledPawn
    (:Is%20Valid
      (Utilities|IsValid (Ability|GetFortAbilitySystemComponent ControlledPawn)
        (:Is%20Valid
          (Utilities|IsValid _returnvalue
            (:Is%20Valid
              (bind _asfort_player_pawn (Utilities|Casting|CastToFortPlayerPawn _returnvalue)
                (:then
                  (if (WaterBody|IsSurfaceSwimming _asfort_player_pawn)
                    (return true)
                    (else
                      (Utilities|IsValid (Vehicle|GetVehicleActor _asfort_player_pawn :!class /Script/FortniteGame.FortPlayerPawn)
                        (:Is%20Valid
                          (bind _returnvalue_2 (Class|FortVehicleInterface|IsinWater(Message) (Vehicle|GetVehicleActor _asfort_player_pawn :!class /Script/FortniteGame.FortPlayerPawn)))
                          (if _returnvalue_2
                            (return true)
                            (else
                              _)))
                        (:Is%20Not%20Valid
                          _))
                      (return false))))
                (:CastFailed
                  (bind _asfort_pawn (Utilities|Casting|CastToFortPawn _returnvalue)
                    (:then
                      (if (WaterBody|IsSurfaceSwimming _asfort_pawn)
                        (return true)
                        (else
                          (return false))))
                    (:CastFailed
                      (bind _asfort_pickup (Utilities|Casting|CastToFortPickup _returnvalue)
                        (:then
                          (if (> (Utilities|Array|Length (Collision|GetOverlappingActors _asfort_pickup "/Script/FortniteGame.FortWaterBodyActor" :!class /Script/Engine.Actor)) 0)
                            (return true)
                            (else
                              (return false))))
                        (:CastFailed
                          (bind _asfort_projectile_base (Utilities|Casting|CastToFortProjectileBase _returnvalue)
                            (:then
                              (if (> (Utilities|Array|Length (Collision|GetOverlappingActors _asfort_projectile_base "/Script/FortniteGame.FortWaterBodyActor" :!class /Script/Engine.Actor)) 0)
                                (return true)
                                (else
                                  (return false))))
                            (:CastFailed
                              (return false))))))))))
            (:Is%20Not%20Valid
              _)))
        (:Is%20Not%20Valid
          _))
      (return false))
    (:Is%20Not%20Valid
      _)))
