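; Behavior tree decorator that checks the pawn type of the actor stored in a chosen blackboard key.
; PerformConditionCheckAI reads that actor through ActorKeyToTest, and LookForPlayer selects a cast to FortPlayerPawn or FortAIPawn.
; In player mode a failed cast returns false and a successful one returns true.
; In AI mode neither cast outcome returns false, so the check as written always ends in true.
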(fn PerformConditionCheckAI
  (params (OwnerController :type AIController%20Object%20Reference) (ControlledPawn :type Pawn%20Object%20Reference))
  (bind _actorkeytotest (Variables|Default|GetActorKeyToTest))
  (bind _returnvalue (AI|BehaviorTree|GetBlackboardValueasActor _actorkeytotest))
  (if (Variables|Default|GetLookForPlayer)
    (Utilities|Casting|CastToFortPlayerPawn _returnvalue
      (:then
        _)
      (:CastFailed
        (return false)))
    (else
      (Utilities|Casting|CastToFortAIPawn _returnvalue
        (:then
          _)
        (:CastFailed
          _))))
  (return true))
