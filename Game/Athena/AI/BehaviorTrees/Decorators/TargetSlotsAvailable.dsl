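; Behavior tree decorator for Athena AI that checks whether a target has enough empty slots for the NPC to slot into.
; It casts the controlled pawn to NPC_Pawn_Parent, caches the pawn and controller, and reads the target actor from a configurable blackboard key.
; For a valid target, the pawn's NPC target slot info answers for the configured slot type.
; A failed cast or invalid target fails the check.
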(fn PerformConditionCheckAI
  (params (OwnerController :type AIController%20Object%20Reference) (ControlledPawn :type Pawn%20Object%20Reference))
  (bind _npc_blackboardkeytotesttargetslotson (Variables|NPCTargetSlots|GetNPC%20BlackboardKeyToTestTargetSlotsOn))
  (bind _returnvalue (AI|BehaviorTree|GetBlackboardValueasActor _npc_blackboardkeytotesttargetslotson))
  ;[1 1 1] See if our Target has enough empty Slots for us to Slot into
  (bind _asnpc_pawn_parent (Utilities|Casting|CastToNPC_Pawn_Parent ControlledPawn)
    (:then
      (bind _output_get (Variables|InternalLogicVariables|SetNPC%20PawnParent _asnpc_pawn_parent))
      (Variables|InternalLogicVariables|SetNPC%20PawnParentController OwnerController)
      (Utilities|IsValid _returnvalue
        (:Is%20Valid
          (return (.NPC%20HasSpaceToSlotOnTarget (Class|NPCPawnParent|NPCTargetSlotInfo _output_get (Variables|NPCTargetSlots|GetTargetSlotType) _returnvalue))))
        (:Is%20Not%20Valid
          _)))
    (:CastFailed
      _))
  (return false))
